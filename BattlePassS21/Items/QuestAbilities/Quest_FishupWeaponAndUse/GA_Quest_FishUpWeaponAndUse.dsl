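; Quest gameplay ability that credits use of a fished-up weapon against qualifying targets.
; On its trigger event, it reads the target pickup's weapon item definition and stores that weapon's gameplay tags.
; It then listens for Event.Damage.Dealt, requiring all stored weapon tags on the instigator and any configured target tag on the victim.
; A match sends the FishedUpWeaponElim stat to the quest manager, though the check is on damage events.
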(event Ability|EventActivateAbilityFromEvent
  (params (EventData :type Gameplay%20Event%20Data%20Structure%20%28by%20ref%29))
  (bind (_eventtag _instigator _target) (Utilities|Struct|BreakGameplayEventData EventData))
  (|CallParentFunction EventData)
  (bind _asfort_pickup (Utilities|Casting|CastToFortPickup _target)
    (:then
      (bind _asfort_weapon_item_definition (Utilities|Casting|CastToFortWeaponItemDefinition (Class|FortItemFunctionLibrary|BreakFortItemEntry (Pickup|GetItemEntry _asfort_pickup 0)))
        (:then
          (Variables|Default|SetWeapon%20Tags (GameplayTags|GetOwnedGameplayTags _asfort_weapon_item_definition :!class /Script/GameplayTags.GameplayTagAssetInterface))
          (bind (_asynctaskproxy _payload) (Ability|Tasks|WaitGameplayEvent "(TagName=\"Event.Damage.Dealt\")" 0 false false)
            (:EventReceived
              (if (and (GameplayTags|HasAllTags (.InstigatorTags (Utilities|Struct|BreakGameplayEventData _payload)) (Variables|Default|GetWeapon%20Tags)) (GameplayTags|HasAnyTags (.TargetTags (Utilities|Struct|BreakGameplayEventData _payload)) (Variables|Default|GetTarget%20Tags)))
                (Quest|SendComplexCustomStatEvent (Variables|GAVars|Data|GetOwningPlayerQuestManager) 0 "(GameplayTags=((TagName=\"Athena.Quests.FishedUpWeaponElim\")))" "(GameplayTags=)" 1)))))
        (:CastFailed
          _)))
    (:CastFailed
      _))
  (Ability|EndAbility))
